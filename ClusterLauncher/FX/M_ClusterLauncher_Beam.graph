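# Unlit additive beam material for the ClusterLauncher effects, flagged for Niagara ribbons.
# A sphere mask on the green UV channel forms a soft band, with parameters for location, radius and hardness, centered by default.
# A power node shapes the band, which is then multiplied by particle alpha and an overall opacity parameter.
# The graph samples no textures, so the beam profile is fully procedural.
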Material /ClusterLauncher/FX/M_ClusterLauncher_Beam
graph {
 properties {
  bAutomaticallySetUsageInEditor=false
  bUsedWithNiagaraRibbons=true
  blendMode=BLEND_Additive
  shadingModel=MSM_Unlit
  usedShadingModels=Unlit
 }
 nodes [
  n1 = SphereMask(
   A=ComponentMask(None=TextureCoordinate()) {g=true}
   B=ScalarParameter() {defaultValue=0.5, parameterName=location}
   Radius=ScalarParameter() {defaultValue=0.25, parameterName=radius}
   Hardness=ScalarParameter() {parameterName=hardness}
  )
  Multiply(A=ScalarParameter() {defaultValue=0.25, parameterName=overall opacity}, B=Multiply(A=ParticleColor().A, B=Power(Base=@n1)))
 ]
}
