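# Niagara targeting beam for the Cluster Launcher, drawing a red laser-like ribbon from the skeletal mesh to User.BeamEnd_0.
# MainRibbon lays partitioned particles along a Hermite spline, with alpha curves showing it near both ends and hiding the middle.
# Glow and Flare sprites sit at the beam end, pulse red over the timeline, and take a camera offset for non-local players.
# Two lens-glint emitters and the ribbon wobble are disabled.
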NiagaraSystem /ClusterLauncher/FX/NS_ClusterLauncher_Targeting
dependencies {
 usedDataInterfaces [
  &NiagaraDataInterfaceCurve
  &NiagaraDataInterfaceColorCurve
  &NiagaraDataInterfaceArrayFloat2
  &NiagaraDataInterfaceArrayFloat3
  &NiagaraDataInterfaceArrayFloat
  /Script/Niagara.NiagaraDataInterfaceOcclusion
  &NiagaraDataInterfaceCamera
  &NiagaraDataInterfaceSkeletalMesh
 ]
 usedDynamicInputs []
 usedModules [
  &SystemState
  &self:SystemSpawnScript.SystemScriptSource.SystemScriptGraph.NiagaraNodeAssignment_9.SetVariables_#1
  &self:SystemSpawnScript.SystemScriptSource.SystemScriptGraph.NiagaraNodeAssignment_8.SetVariables_#2
  &self:SystemSpawnScript.SystemScriptSource.SystemScriptGraph.NiagaraNodeAssignment_1.SetVariables_#3
 ]
}
emitters {
 Flare {
  emitterUpdateScript [
   &EmitterState(Life Cycle Mode=::NewEnumerator1, Loop Behavior=::NewEnumerator1, Loop Duration Mode=::NewEnumerator1, UseLoopDelay=false)
   &SpawnBurst_Instantaneous(Spawn Count=3, Use Spawn Probability=false, UseLoopCountLimit=false)
  ]
  particleSpawnScript [
   &PartitionParticles(Number Of Partitions=3)
   &InitializeParticle(
    Lifetime=0.1
    Sprite Rotation Mode=::NewEnumerator0
    Sprite Size=&SelectVector2DFromArray(
     Direct Array Index=$Particles.PartitionParticles.Partition
     Vector 2D Selection Array=DataInterface() {propertyValues {FloatData [<301.5 10>, <167.5 10>, <125 10>]}}
    )
    Sprite Size Mode=::NewEnumerator4
    UsePositionOffset=false
    Write Material Random=true
   )
  ]
  particleUpdateScript [
   &ParticleState(Kill Particles When Lifetime Has Elapsed=false, Let Infinitely Lived Particles Die When Emitter Deactivates=true)
   &SkeletalMeshLocation(
    Bone Sampling Mode=::NewEnumerator4
    MaskSpawnGroup=false
    Sampled Position Offset=&SelectVectorFromArray(
     Direct Array Index=$Particles.PartitionParticles.Partition
     Vector Selection Array=DataInterface() {propertyValues {FloatData [<0 0 1>, <0 0 0>, <0 0 -1>]}}
    )
    Skeletal Mesh=$System.SkeletalMesh
   ) {enabled=false}
   &ShapeLocation(MaskSpawnGroup=false, Override Seed=false, Shape Origin=$User.BeamEnd_0, Sphere Radius=0, UseOwnerScale=true)
   &DynamicMaterialParameters(
    Index 0 Param 1=0.2
    Index 0 Param 2=5
    Index 0 Param 3=1
    Index 0 Param 4=1
    Module.Index0Param0Write=true
    Module.Index0Param1Write=true
    Module.Index0Param2Write=true
    Module.Index0Param3Write=true
    Module.Index1Param0Write=true
    Module.Index1Param1Write=true
    Module.Index1Param2Write=true
    Module.Index1Param3Write=true
    Write Parameter Index 1=true
   )
   &CameraOffset(Camera Offset Amount=&MakeCustomFloatFromBool(Bool=$User.IsLocalPlayer, True Float=0, False Float=75))
   &ScaleSpriteSize(Scale Sprite Size Mode=::NewEnumerator0, Uniform Scale Factor=0.7)
   &ScaleColor(
    Scale Alpha=&FNTimeOfDayScalar(Night Scale=0.2, Morning Scale=0.3, Day Scale=0.4, Evening Scale=0.3)
    Scale RGB=&VectorFromFloat(Value=&SelectFloatFromArray(Direct Array Index=$Particles.PartitionParticles.Partition, Float Selection Array=DataInterface() {propertyValues {FloatData [6,4,4]}}))
    ScaleA=true
    ScaleRGB=true
   )
   &ScaleColor(
    Scale RGB=&MakeVectorFromLinearColorRGB(Color=&ColorFromCurve(
      ColorCurve=DataInterface() {propertyValues {
        AlphaCurve {keys [<at 0 1 RCIM_Cubic>]}
        BlueCurve {keys [<at 0 0 RCIM_Cubic>,<at 0.5178 0 RCIM_Cubic>,<at 1 0>]}
        GreenCurve {keys [<at 0 0 RCIM_Cubic>,<at 0.5178 0 RCIM_Cubic>,<at 1 0>]}
        RedCurve {keys [<at 0 1 RCIM_Cubic>,<at 0.5178 10 RCIM_Cubic>,<at 1 1>]}
       }}
      CurveIndex=$System.Timeline.NormalizedTime
     ))
    ScaleA=true
    ScaleRGB=true
   ) {module=ScaleColor002}
   &ScaleColor(
    Scale Alpha=&FloatFromCurve(FloatCurve=DataInterface() {propertyValues {Curve {keys [{interpMode=RCIM_Cubic, tangentMode=RCTM_User, time=0, value=0}, {interpMode=RCIM_Cubic, tangentMode=RCTM_User, time=0.5, value=1}]}}})
    ScaleA=true
    ScaleRGB=false
   ) {module=ScaleColor001}
  ]
  rendererClasses [&NiagaraSpriteRendererProperties]
  renderers [{rendererClass=&NiagaraSpriteRendererProperties}]
 }
 GlintLensFlarePlus {
  bEnabled=false
  simTarget=GPUComputeSim
  emitterUpdateScript [
   &EmitterState(Life Cycle Mode=::NewEnumerator1, Loop Duration=1, UseLoopDelay=false)
   &SpawnBurst_Instantaneous(Use Spawn Probability=false, UseLoopCountLimit=false)
  ]
  particleSpawnScript [
   &PartitionParticles()
   &InitializeParticle(
    Color=<rgba 0.6273 0.8504 1.3 1>
    Sprite Size Mode=::NewEnumerator3
    Uniform Sprite Size=250
    UsePositionOffset=false
    Write Material Random=true
   )
   &CameraOffset(Camera Offset Amount=15) {enabled=false}
  ]
  particleUpdateScript [
   /SetVariables(Particles.DistanceSizeScalar:NiagaraFloat=&FloatFromCurve(
     FloatCurve=DataInterface() {propertyValues {Curve {keys [
         {interpMode=RCIM_Cubic, tangentMode=RCTM_User, time=0, value=0.1}
         {arriveTangent=9.568, interpMode=RCIM_Cubic, leaveTangent=9.568, tangentMode=RCTM_User, time=0.05, value=0.8}
         {
          arriveTangent=1.4123
          arriveTangentWeight=0.0167
          interpMode=RCIM_Cubic
          leaveTangent=1.4123
          leaveTangentWeight=0.5781
          tangentMode=RCTM_User
          time=0.4
          value=2
         }
         {arriveTangentWeight=0.3181, interpMode=RCIM_Cubic, tangentMode=RCTM_User, time=1, value=2.5}
        ]}}}
     CurveIndex=$Emitter.Scalability.DistanceFraction
    )) {isSetParameters=true, module=SetVariables_#4}
   &ParticleState(Kill Particles When Lifetime Has Elapsed=false, Let Infinitely Lived Particles Die When Emitter Deactivates=true)
   &SkeletalMeshLocation(Bone Sampling Mode=::NewEnumerator4, MaskSpawnGroup=false, Skeletal Mesh=$System.SkeletalMesh)
   &self:ScratchModule_02(OcclusionSize=20) {module=ScratchModule_02}
   /SetVariables(Particles.CameraScale:NiagaraFloat=&Multiply_Float(
     A=1
     B=&MakeFloatFromVector(
      Vector=&ReturnCameraVector(Camera Vectors=::NewEnumerator4, Coordinate Space=::Simulation, Origin Position=$Particles.Position)
      Channel=::NewEnumerator0
     )
    )) {isSetParameters=true, module=SetVariables_#5}
   &ScaleSpriteSize(
    Scale Sprite Size Mode=::NewEnumerator0
    Uniform Scale Factor=&Multiply_Float(
     A=&Lerp_Float(
      A=0.15
      B=1.5
      Alpha=&FloatFromCurve(
       FloatCurve=DataInterface() {propertyValues {Curve {keys [
           {interpMode=RCIM_Cubic, tangentMode=RCTM_User, time=0.7, value=0}
           {arriveTangent=2.314, interpMode=RCIM_Cubic, leaveTangent=2.314, tangentMode=RCTM_User, time=0.9, value=0.15}
           {arriveTangent=11.09, interpMode=RCIM_Cubic, leaveTangent=11.09, tangentMode=RCTM_User, time=1, value=1}
          ]}}}
       CurveIndex=$Particles.CameraScale
      )
     )
     B=$Particles.DistanceSizeScalar
    )
   )
   /SetVariables(Particles.FlareScaleAlt:Vector2f=&Multiply_Vector2D_ByFloat(
     Float=&Lerp_Float(
      A=0.15
      B=1.5
      Alpha=&FloatFromCurve(
       FloatCurve=DataInterface() {propertyValues {Curve {keys [
           {interpMode=RCIM_Cubic, tangentMode=RCTM_User, time=0.7, value=0}
           {arriveTangent=3.3224, interpMode=RCIM_Cubic, leaveTangent=3.3224, tangentMode=RCTM_User, time=0.9, value=0.35}
           {arriveTangent=11.09, interpMode=RCIM_Cubic, leaveTangent=11.09, tangentMode=RCTM_User, time=1, value=1}
          ]}}}
       CurveIndex=$Particles.CameraScale
      )
     )
     Vector2D=&Multiply_Vector2D_ByFloat(Float=&Multiply_Float(A=0.6, B=$Particles.DistanceSizeScalar), Vector2D=<500 250>)
    )) {isSetParameters=true, module=SetVariables_#6}
   &ScaleColor(Scale Alpha=$Particles.Visibility, ScaleA=true, ScaleRGB=false) {module=ScaleColor002}
   &ScaleColor(
    Scale Alpha=&FloatFromCurve(
     FloatCurve=DataInterface() {propertyValues {Curve {keys [
         {tangentMode=RCTM_User, time=0, value=0}
         {arriveTangent=0.4636, interpMode=RCIM_Cubic, leaveTangent=0.4636, tangentMode=RCTM_User, time=0.8, value=0}
         {arriveTangent=0.1158, interpMode=RCIM_Cubic, leaveTangent=0.1158, tangentMode=RCTM_User, time=1, value=1}
        ]}}}
     CurveIndex=$Particles.CameraScale
     Scale Curve=&FloatFromCurve(
      FloatCurve=DataInterface() {propertyValues {Curve {keys [
          {interpMode=RCIM_Cubic, tangentMode=RCTM_User, time=0, value=0.25}
          {interpMode=RCIM_Cubic, tangentMode=RCTM_User, time=0.1, value=1}
          {arriveTangentWeight=1.0285, interpMode=RCIM_Cubic, tangentMode=RCTM_User, time=1, value=1}
         ]}}}
      CurveIndex=$Emitter.Scalability.DistanceFraction
     )
    )
    ScaleA=true
    ScaleRGB=false
   )
   &SpriteFacingAndAlignment(
    Sprite Alignment=&Lerp_Vector(
     A=&Lerp_Vector(
      A=<-1 0 1>
      B=<1 0 1>
      Alpha=&FloatFromCurve(
       FloatCurve=DataInterface() {propertyValues {Curve {keys [
           {interpMode=RCIM_Cubic, tangentMode=RCTM_User, time=0.3, value=0}
           {arriveTangentWeight=0.7423, interpMode=RCIM_Cubic, tangentMode=RCTM_User, time=1, value=1}
          ]}}}
       CurveIndex=$Particles.CameraScale
      )
     )
     B=<0 0 1>
     Alpha=&FloatFromCurve(
      FloatCurve=DataInterface() {propertyValues {Curve {keys [
          {interpMode=RCIM_Cubic, tangentMode=RCTM_User, time=0, value=0}
          {arriveTangent=0.1269, interpMode=RCIM_Cubic, leaveTangent=0.1269, tangentMode=RCTM_User, time=0.9, value=1}
          {arriveTangentWeight=0.7423, interpMode=RCIM_Cubic, tangentMode=RCTM_User, time=1, value=1}
         ]}}}
      CurveIndex=$Particles.CameraScale
     )
    )
    WriteAlignment=true
    WriteFacing=true
   )
   &ScaleColor(
    Scale Alpha=&FNTimeOfDayScalar(Night Scale=0.25, Morning Scale=0.5, Evening Scale=0.5)
    ScaleA=true
    ScaleRGB=false
   ) {module=ScaleColor001}
  ]
  rendererClasses [&NiagaraSpriteRendererProperties]
  renderers [
   {rendererClass=&NiagaraSpriteRendererProperties}
   {rendererClass=&NiagaraSpriteRendererProperties, rendererIndex=1}
  ]
 }
 GlintLensFlare_CPU {
  bEnabled=false
  emitterUpdateScript [
   &EmitterState(Life Cycle Mode=::NewEnumerator1, Loop Duration=1, UseLoopDelay=false)
   &SpawnBurst_Instantaneous(Use Spawn Probability=false, UseLoopCountLimit=false)
  ]
  particleSpawnScript [
   &PartitionParticles()
   &InitializeParticle(
    Color=<rgba 0.6273 0.8504 1.3 1>
    Sprite Size Mode=::NewEnumerator3
    Uniform Sprite Size=250
    UsePositionOffset=false
    Write Material Random=true
   )
  ]
  particleUpdateScript [
   /SetVariables(Particles.DistanceSizeScalar:NiagaraFloat=&FloatFromCurve(
     FloatCurve=DataInterface() {propertyValues {Curve {keys [
         {interpMode=RCIM_Cubic, tangentMode=RCTM_User, time=0, value=0.1}
         {arriveTangent=9.568, interpMode=RCIM_Cubic, leaveTangent=9.568, tangentMode=RCTM_User, time=0.05, value=0.8}
         {
          arriveTangent=1.4123
          arriveTangentWeight=0.0167
          interpMode=RCIM_Cubic
          leaveTangent=1.4123
          leaveTangentWeight=0.5781
          tangentMode=RCTM_User
          time=0.4
          value=2
         }
         {arriveTangentWeight=0.3181, interpMode=RCIM_Cubic, tangentMode=RCTM_User, time=1, value=2.5}
        ]}}}
     CurveIndex=$Emitter.Scalability.DistanceFraction
    )) {isSetParameters=true, module=SetVariables_#4}
   &ParticleState(Kill Particles When Lifetime Has Elapsed=false, Let Infinitely Lived Particles Die When Emitter Deactivates=true)
   &SkeletalMeshLocation(Bone Sampling Mode=::NewEnumerator4, MaskSpawnGroup=false, Skeletal Mesh=$System.SkeletalMesh)
   &self:Opponent_Visibility() {module=Opponent_Visibility}
   /SetVariables(Particles.CameraScale:NiagaraFloat=&Multiply_Float(
     A=1
     B=&MakeFloatFromVector(
      Vector=&ReturnCameraVector(Camera Vectors=::NewEnumerator4, Coordinate Space=::Simulation, Origin Position=$Particles.Position)
      Channel=::NewEnumerator0
     )
    )) {isSetParameters=true, module=SetVariables_#5}
   &ScaleSpriteSize(
    Scale Sprite Size Mode=::NewEnumerator0
    Uniform Scale Factor=&Multiply_Float(
     A=&Lerp_Float(
      A=0.15
      B=1.5
      Alpha=&FloatFromCurve(
       FloatCurve=DataInterface() {propertyValues {Curve {keys [
           {interpMode=RCIM_Cubic, tangentMode=RCTM_User, time=0.7, value=0}
           {arriveTangent=2.314, interpMode=RCIM_Cubic, leaveTangent=2.314, tangentMode=RCTM_User, time=0.9, value=0.15}
           {arriveTangent=11.09, interpMode=RCIM_Cubic, leaveTangent=11.09, tangentMode=RCTM_User, time=1, value=1}
          ]}}}
       CurveIndex=$Particles.CameraScale
      )
     )
     B=$Particles.DistanceSizeScalar
    )
   )
   /SetVariables(Particles.FlareScaleAlt:Vector2f=&Multiply_Vector2D_ByFloat(
     Float=&Lerp_Float(
      A=0.15
      B=1.5
      Alpha=&FloatFromCurve(
       FloatCurve=DataInterface() {propertyValues {Curve {keys [
           {interpMode=RCIM_Cubic, tangentMode=RCTM_User, time=0.7, value=0}
           {arriveTangent=3.3224, interpMode=RCIM_Cubic, leaveTangent=3.3224, tangentMode=RCTM_User, time=0.9, value=0.35}
           {arriveTangent=11.09, interpMode=RCIM_Cubic, leaveTangent=11.09, tangentMode=RCTM_User, time=1, value=1}
          ]}}}
       CurveIndex=$Particles.CameraScale
      )
     )
     Vector2D=&Multiply_Vector2D_ByFloat(Float=&Multiply_Float(A=0.6, B=$Particles.DistanceSizeScalar), Vector2D=<500 250>)
    )) {isSetParameters=true, module=SetVariables_#6}
   &ScaleColor(Scale Alpha=$Particles.Visibility, ScaleA=true, ScaleRGB=false) {module=ScaleColor002}
   &ScaleColor(
    Scale Alpha=&FloatFromCurve(
     FloatCurve=DataInterface() {propertyValues {Curve {keys [
         {tangentMode=RCTM_User, time=0, value=0}
         {arriveTangent=0.4636, interpMode=RCIM_Cubic, leaveTangent=0.4636, tangentMode=RCTM_User, time=0.8, value=0}
         {arriveTangent=0.1158, interpMode=RCIM_Cubic, leaveTangent=0.1158, tangentMode=RCTM_User, time=1, value=1}
        ]}}}
     CurveIndex=$Particles.CameraScale
     Scale Curve=&FloatFromCurve(
      FloatCurve=DataInterface() {propertyValues {Curve {keys [
          {interpMode=RCIM_Cubic, tangentMode=RCTM_User, time=0, value=0.25}
          {interpMode=RCIM_Cubic, tangentMode=RCTM_User, time=0.1, value=1}
          {arriveTangentWeight=1.0285, interpMode=RCIM_Cubic, tangentMode=RCTM_User, time=1, value=1}
         ]}}}
      CurveIndex=$Emitter.Scalability.DistanceFraction
     )
    )
    ScaleA=true
    ScaleRGB=false
   )
   &SpriteFacingAndAlignment(
    Sprite Alignment=&Lerp_Vector(
     A=&Lerp_Vector(
      A=<-1 0 1>
      B=<1 0 1>
      Alpha=&FloatFromCurve(
       FloatCurve=DataInterface() {propertyValues {Curve {keys [
           {interpMode=RCIM_Cubic, tangentMode=RCTM_User, time=0.3, value=0}
           {arriveTangentWeight=0.7423, interpMode=RCIM_Cubic, tangentMode=RCTM_User, time=1, value=1}
          ]}}}
       CurveIndex=$Particles.CameraScale
      )
     )
     B=<0 0 1>
     Alpha=&FloatFromCurve(
      FloatCurve=DataInterface() {propertyValues {Curve {keys [
          {interpMode=RCIM_Cubic, tangentMode=RCTM_User, time=0, value=0}
          {arriveTangent=0.1269, interpMode=RCIM_Cubic, leaveTangent=0.1269, tangentMode=RCTM_User, time=0.9, value=1}
          {arriveTangentWeight=0.7423, interpMode=RCIM_Cubic, tangentMode=RCTM_User, time=1, value=1}
         ]}}}
      CurveIndex=$Particles.CameraScale
     )
    )
    WriteAlignment=true
    WriteFacing=true
   ) {module=SpriteFacingAndAlignment001}
   &ScaleColor(
    Scale Alpha=&FNTimeOfDayScalar(Night Scale=0.25, Morning Scale=0.5, Evening Scale=0.5)
    ScaleA=true
    ScaleRGB=false
   ) {module=ScaleColor001}
  ]
  rendererClasses [&NiagaraSpriteRendererProperties]
  renderers [
   {rendererClass=&NiagaraSpriteRendererProperties}
   {rendererClass=&NiagaraSpriteRendererProperties, rendererIndex=1}
  ]
 }
 Glow {
  emitterUpdateScript [
   &EmitterState(Life Cycle Mode=::NewEnumerator1, Loop Behavior=::NewEnumerator1, Loop Duration Mode=::NewEnumerator1, UseLoopDelay=false)
   &SpawnBurst_Instantaneous(Use Spawn Probability=false, UseLoopCountLimit=false)
  ]
  particleSpawnScript [
   &InitializeParticle(Lifetime=0.1, Sprite Size Mode=::NewEnumerator3, Uniform Sprite Size=75, UsePositionOffset=false, Write Material Random=true)
   /SetVariables(Particles.RoundFlareSize:Vector2f=<36 36>) {isSetParameters=true}
  ]
  particleUpdateScript [
   &ParticleState(Kill Particles When Lifetime Has Elapsed=false, Let Infinitely Lived Particles Die When Emitter Deactivates=true)
   &SkeletalMeshLocation(Bone Sampling Mode=::NewEnumerator4, MaskSpawnGroup=false, Skeletal Mesh=$System.SkeletalMesh) {enabled=false}
   &ShapeLocation(MaskSpawnGroup=false, Override Seed=false, Shape Origin=$User.BeamEnd_0, Sphere Radius=0, UseOwnerScale=true)
   &ScaleSpriteSize(Scale Sprite Size Mode=::NewEnumerator0, Uniform Scale Factor=0.7)
   &CameraOffset(Camera Offset Amount=&MakeCustomFloatFromBool(Bool=$User.IsLocalPlayer, True Float=0, False Float=75))
   &ScaleColor(
    Scale RGB=&MakeVectorFromLinearColorRGB(Color=&ColorFromCurve(
      ColorCurve=DataInterface() {propertyValues {
        AlphaCurve {keys [<at 0 1 RCIM_Cubic>]}
        BlueCurve {keys [<at 0 0 RCIM_Cubic>,<at 0.5178 0 RCIM_Cubic>,<at 1 0>]}
        GreenCurve {keys [<at 0 0 RCIM_Cubic>,<at 0.5178 0 RCIM_Cubic>,<at 1 0>]}
        RedCurve {keys [<at 0 1 RCIM_Cubic>,<at 0.5178 10 RCIM_Cubic>,<at 1 1>]}
       }}
      CurveIndex=$System.Timeline.NormalizedTime
     ))
    ScaleA=true
    ScaleRGB=true
   ) {module=ScaleColor001}
   &ScaleColor(Scale RGB=&VectorFromFloat(Value=20), ScaleA=false, ScaleRGB=true)
   &DynamicMaterialParameters(
    Index 0 Param 1=0.2
    Index 0 Param 2=5
    Index 0 Param 3=1
    Index 0 Param 4=1
    Module.Index0Param0Write=true
    Module.Index0Param1Write=true
    Module.Index0Param2Write=true
    Module.Index0Param3Write=true
    Module.Index1Param0Write=true
    Module.Index1Param1Write=true
    Module.Index1Param2Write=true
    Module.Index1Param3Write=true
    Write Parameter Index 1=true
   )
   /SetVariables(Particles.RoundFlareColor:LinearColor=&Multiply_LinearColorByFloat(
     Linear Color=$Particles.Color
     Float=&FloatFromCurve(
      FloatCurve=DataInterface() {propertyValues {Curve {keys [
          {arriveTangent=5.8312, interpMode=RCIM_Cubic, leaveTangent=5.8312, tangentMode=RCTM_User, time=0, value=0}
          {interpMode=RCIM_Cubic, tangentMode=RCTM_User, time=0.5, value=1}
         ]}}}
      Scale Curve=0.25
     )
    )) {isSetParameters=true}
   /SetVariables(Particles.DynMatRoundFlare:Vector4f=<1 0.1 1 0>) {isSetParameters=true}
  ]
  rendererClasses [&NiagaraSpriteRendererProperties]
  renderers [
   {rendererClass=&NiagaraSpriteRendererProperties}
   {rendererClass=&NiagaraSpriteRendererProperties, rendererIndex=1}
  ]
 }
 MainRibbon {
  emitterSpawnScript [/Game/Effects/Niagara/Modules/Partition/FN_PartitionAttributes(Number of Bone Sets=1, Particles Per Partition=15)]
  emitterUpdateScript [
   &EmitterState(Life Cycle Mode=::NewEnumerator1, Loop Behavior=::NewEnumerator1, Loop Duration Mode=::NewEnumerator1, UseLoopDelay=false)
   &SpawnBurst_Instantaneous(Spawn Count=$Emitter.FN_PartitionAttributes.Spawn Total, Use Spawn Probability=false)
   /SetVariables(
    Emitter.CoreWidthPhase:NiagaraFloat=&Lerp_Float(
     A=0.125
     B=0.125
     Alpha=&Waveform("Waveform [1]"=::NewEnumerator0, "[1] Phase Scale"=10, "[1] Frequency"=1, "[1] Amplitude Min/Max"=<0 1>, "[1] Bias"=-0.5)
    )
    Emitter.CosPhaseRand:NiagaraFloat=&RandomRangeFloat(Minimum=1, Maximum=5, Evaluation Type=::NewEnumerator1)
   ) {enabled=false, isSetParameters=true}
  ]
  particleSpawnScript [
   &InitializeRibbon(
    Color=<rgba 1 0.0277 0 1>
    Lifetime=0.1
    Module.Write Color=true
    Module.Write Lifetime=true
    Module.Write Mass=false
    Module.Write Material Random=true
    Module.Write Position=false
    Module.Write RibbonTwist=false
    Module.Write RibbonWidth=true
    Ribbon Width=25
   )
   /Game/Effects/Niagara/Modules/Partition/FN_Particle_partition_v5/FN_Particle_Partition_v5(Enable Ribbon ID=true)
  ]
  particleUpdateScript [
   &ParticleState(Kill Particles When Lifetime Has Elapsed=false, Let Infinitely Lived Particles Die When Emitter Deactivates=true)
   &SkeletalMeshLocation(Bone Sampling Mode=::NewEnumerator4, MaskSpawnGroup=false, Skeletal Mesh=$System.SkeletalMesh)
   /SetVariables(Particles.BeamLength:NiagaraFloat=&VectorLength(Vector=&Subtract_Vector(
      A=&TransformPosition(Position=$User.BeamEnd_0, Source Space=::World, Destination Space=::Local)
      B=$Particles.SkeletalMeshLocation.SampledPosition
     ))) {isSetParameters=true}
   /Game/Effects/Niagara/Modules/Curves/FN_HermiteCubicSpline_v1/FN_HermiteCubicSpline(
    Position Blend:NiagaraFloat=$Particles.Initial.Partition.NormalizedID
    Position End:Vector3f=&AddVectorToPosition(Position=$Particles.SkeletalMeshLocation.SampledPosition, Vector=<0 0 0>)
    Position Start:Vector3f=/Niagara/DynamicInputs/Transforms/TransformPosition(Position=Linked() {linkedVariable {name=User.BeamEnd_0, type=NiagaraPosition}}, Source Space=::World, Destination Space=::Local)
    Write Particles Position=true
   )
   &FN_HermiteCubicSpline(
    Position Blend:NiagaraFloat=$Particles.Initial.Partition.NormalizedID
    Position End:Vector3f=&TransformPosition(Position=Linked() {linkedVariable {name=User.BeamEnd_0, type=NiagaraPosition}}, Source Space=::World, Destination Space=::Local)
    Write Particles Position=true
   ) {module=FN_HermiteCubicSpline001}
   /SetVariables(
    Particles.PosLocal:NiagaraPosition=&Add_Vector(
     A=$Output.FN_HermiteCubicSpline.Position
     B=&Multiply_VectorByFloat(
      Vector=&RotateVectorByQuaternion(
       Vector To Rotate=&MakeVector(
        X=0
        Y=&Waveform(
         Global Input Position=10
         Global Amplitude Scale=20
         "Waveform [1]"=::NewEnumerator0
         "[1] Frequency"=5
         "[1] Amplitude Min/Max"=<0 1>
         "[1] Bias"=-0.5
         "[2] Phase"=$Emitter.Age
         "[3] Phase"=$Emitter.Age
        )
        Z=&Waveform(
         Global Input Position=10
         Global Amplitude Scale=20
         "Waveform [1]"=::NewEnumerator1
         "[1] Frequency"=1
         "[1] Amplitude Min/Max"=<0 1>
         "[1] Bias"=-0.5
         "[2] Phase"=$Emitter.Age
         "[3] Phase"=$Emitter.Age
        )
       )
       Quaternion=$Particles.BeamOrientation
      )
      Float=$Particles.WobbleOverLength
     )
    )
    Particles.PosOtherPlayer:NiagaraPosition=&Add_Vector(
     A=$Output.FN_HermiteCubicSpline.Position
     B=&Multiply_VectorByFloat(
      Vector=&RotateVectorByQuaternion(Vector To Rotate=&MakeVector(X=0, Y=0, Z=0), Quaternion=$Particles.BeamOrientation)
      Float=$Particles.WobbleOverLength
     )
    )
   ) {enabled=false, isSetParameters=true}
   &ScaleRibbonWidth(Ribbon Width Scale=&FloatFromCurve(
     FloatCurve=DataInterface() {propertyValues {Curve {keys [
         {interpMode=RCIM_Cubic, leaveTangentWeight=0.7456, tangentMode=RCTM_User, time=0, value=1}
         {arriveTangentWeight=0.0667, interpMode=RCIM_Cubic, leaveTangentWeight=0.1667, tangentMode=RCTM_User, time=0.6, value=1}
         {arriveTangentWeight=0.1297, interpMode=RCIM_Cubic, tangentMode=RCTM_User, time=1, value=0.7}
        ]}}}
     CurveIndex=$Particles.Initial.Partition.NormalizedID
     Scale Curve=&Lerp_Float(
      A=0.5
      B=1
      Alpha=&FloatFromCurve(
       FloatCurve=DataInterface() {propertyValues {Curve {keys [
           {interpMode=RCIM_Cubic, tangentMode=RCTM_User, time=0, value=0}
           {arriveTangent=3.4966, interpMode=RCIM_Cubic, leaveTangent=3.4966, tangentMode=RCTM_User, time=0.8, value=1}
           {arriveTangent=-13.2607, interpMode=RCIM_Cubic, leaveTangent=-13.2607, tangentMode=RCTM_User, time=1, value=0}
          ]}}}
       CurveIndex=$Particles.Timeline.NormalizedTime
      )
     )
    ))
   &ScaleColor(
    Scale Alpha=&FloatFromCurve(
     FloatCurve=DataInterface() {propertyValues {Curve {keys [
         {arriveTangent=20.8537, interpMode=RCIM_Cubic, leaveTangent=20.8537, leaveTangentWeight=12.2844, tangentMode=RCTM_User, time=0, value=0}
         {arriveTangentWeight=0.1, interpMode=RCIM_Cubic, leaveTangentWeight=0.0667, tangentMode=RCTM_User, time=0.1, value=1}
         {arriveTangentWeight=0.1795, interpMode=RCIM_Cubic, leaveTangentWeight=0.1795, tangentMode=RCTM_User, time=0.2, value=0}
         {arriveTangentWeight=0.1491, interpMode=RCIM_Cubic, leaveTangentWeight=0.2236, tangentMode=RCTM_User, time=0.7, value=0}
         {arriveTangentWeight=0.2333, interpMode=RCIM_Cubic, leaveTangentWeight=0.3067, tangentMode=RCTM_Break, time=0.9, value=1}
         {arriveTangent=-15.8223, arriveTangentWeight=0.1667, interpMode=RCIM_Cubic, leaveTangent=-15.8223, tangentMode=RCTM_User, time=1, value=0}
        ]}}}
     CurveIndex=$Particles.Initial.Partition.NormalizedID
    )
    Scale RGB=&VectorFromFloat(Value=5)
    ScaleA=true
    ScaleRGB=true
   )
   &ScaleColor(
    Scale Alpha=&FloatFromCurve(
     FloatCurve=DataInterface() {propertyValues {Curve {keys [
         {arriveTangent=20.8537, interpMode=RCIM_Cubic, leaveTangent=20.8537, leaveTangentWeight=12.2844, tangentMode=RCTM_User, time=0, value=0}
         {arriveTangentWeight=0.1, interpMode=RCIM_Cubic, leaveTangentWeight=0.0667, tangentMode=RCTM_User, time=0.1, value=1}
         {arriveTangentWeight=0.2333, interpMode=RCIM_Cubic, leaveTangentWeight=0.3067, tangentMode=RCTM_Break, time=0.8, value=1}
         {arriveTangent=-15.8223, arriveTangentWeight=0.1667, interpMode=RCIM_Cubic, leaveTangent=-15.8223, tangentMode=RCTM_User, time=1, value=0}
        ]}}}
     CurveIndex=$Particles.Initial.Partition.NormalizedID
    )
    Scale RGB=&VectorFromFloat(Value=&Lerp_Float(
      A=2.5
      B=5
      Alpha=&Waveform("Waveform [1]"=::NewEnumerator0, "[1] Phase Scale"=3, "[1] Frequency"=1, "[1] Amplitude Min/Max"=<0 1>)
     ))
    ScaleA=true
    ScaleRGB=true
   ) {enabled=false, module=ScaleColor002}
   /SetVariables(Particles.DebugFloat:NiagaraFloat=$User.Overheat) {enabled=false, isSetParameters=true}
  ]
  rendererClasses [&NiagaraRibbonRendererProperties]
  renderers [{rendererClass=&NiagaraRibbonRendererProperties}]
 }
}
summary {
 systemName=NS_ClusterLauncher_Targeting
 emitters [
  {emitterName=MainRibbon, rendererClasses [&NiagaraRibbonRendererProperties]}
  {emitterName=Glow, rendererClasses [&NiagaraSpriteRendererProperties]}
  {emitterName=Flare, rendererClasses [&NiagaraSpriteRendererProperties]}
  {bEnabled=false, emitterName=GlintLensFlarePlus, simTarget=GPUComputeSim, rendererClasses [&NiagaraSpriteRendererProperties]}
  {bEnabled=false, emitterName=GlintLensFlare_CPU, rendererClasses [&NiagaraSpriteRendererProperties]}
 ]
 userVariables [
  User.BeamEnd_0 {type=NiagaraPosition, description="", defaultValue {}}
  User.BeamStart_0_0 {type=Vector3f, description="", defaultValue {struct=Vector3f, value=<0 0 0>}}
  User.ImpactNormal {type=Vector3f, description="", defaultValue {struct=Vector3f, value=<0 0 1>}}
  User.IsLocalPlayer {type=NiagaraBool}
 ]
}
